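# Sound cue for generic sword impact sounds on melee weapons.
# A random node picks one of four equally weighted sword impact waves, then a modulator node feeds the output.
# It uses small melee weapon attenuation, the melee weapon sound class and a player melee weapon concurrency set.
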SoundCue /Game/Sounds/Fort_Weapons/Swords/Sword_Impact_Generic_Cue
root=Modulator_0
nodes [
 {id=Modulator_0, children [Random_0]}
 {id=Random_0, children [WavePlayer_1,WavePlayer_2,WavePlayer_3,WavePlayer_4], properties {weights [1,1,1,1]}}
 {id=WavePlayer_1, properties {soundWaveAssetPtr=&/sword_impact_01}}
 {id=WavePlayer_2, properties {soundWaveAssetPtr=&/sword_impact_02}}
 {id=WavePlayer_3, properties {soundWaveAssetPtr=&/sword_impact_03}}
 {id=WavePlayer_4, properties {soundWaveAssetPtr=&/sword_impact_04}}
]
properties {
 attenuationSettings=&Fort_Weapon_Melee_Sml_Att
 maxDistance=2800
 soundClassObject=&Fort_WeaponMelee_PSM
 volumeMultiplier=1
 concurrencyOverrides {resolutionRule=PreventNew}
 concurrencySet [&Player_Melee_Weapon_Conc]
}
