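# Sound cue for the destruction of a tier 01 sword melee weapon.
# A random node picks with equal weight between two destroy waves borrowed from the axe sounds.
# It uses small melee weapon attenuation and the melee weapon sound class, with a concurrency override that prevents new instances once the limit is reached.
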SoundCue /Game/Sounds/Fort_Weapons/Swords/Sword_T01_Destroy_Cue
root=Random_4
nodes [
 {id=Random_4, children [WavePlayer_20, WavePlayer_21], properties {weights [1,1]}}
 {id=WavePlayer_20, properties {soundWaveAssetPtr=/Game/Sounds/Fort_Weapons/Axes/axe_destroy_01}}
 {id=WavePlayer_21, properties {soundWaveAssetPtr=/Game/Sounds/Fort_Weapons/Axes/axe_destroy_02}}
]
properties {
 attenuationSettings=&Fort_Weapon_Melee_Sml_Att
 bOverrideConcurrency=true
 maxDistance=2800
 soundClassObject=&Fort_WeaponMelee_PSM
 volumeMultiplier=1
 concurrencyOverrides {resolutionRule=PreventNew}
}
