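# Sound cue for sword swing slashes among Fortnite's melee weapon sounds.
# A quality-level node selects between an equally weighted random pool of five slash waves and a smaller three-wave pool for the next level, leaving its other two slots empty.
# Each pool feeds a modulator that varies pitch and slightly lowers volume at a fixed level.
# Shared small-melee attenuation, melee sound class and player melee concurrency assets govern playback.
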SoundCue /Game/Sounds/Fort_Weapons/Swords/Weapon_M_Sword_Swing_Cue
root=QualityLevel_0
nodes [
 {id=Modulator_1, children [Random_1], properties {pitchMax=1.1, volumeMax=0.8, volumeMin=0.8}}
 {id=Modulator_2, children [Random_2], properties {pitchMax=1.1, volumeMax=0.8, volumeMin=0.8}}
 {id=QualityLevel_0, children [Modulator_1,Modulator_2,null,null]}
 {id=Random_1, children [WavePlayer_5,WavePlayer_6,WavePlayer_7,WavePlayer_8,WavePlayer_9], properties {weights [1,1,1,1,1]}}
 {id=Random_2, children [WavePlayer_10, WavePlayer_11, WavePlayer_12], properties {weights [1,1,1]}}
 {id=WavePlayer_10, properties {soundWaveAssetPtr=&/sword_slash_01}}
 {id=WavePlayer_11, properties {soundWaveAssetPtr=&/sword_slash_02}}
 {id=WavePlayer_12, properties {soundWaveAssetPtr=&/sword_slash_03}}
 {id=WavePlayer_5, properties {soundWaveAssetPtr=&sword_slash_01}}
 {id=WavePlayer_6, properties {soundWaveAssetPtr=&sword_slash_02}}
 {id=WavePlayer_7, properties {soundWaveAssetPtr=&sword_slash_03}}
 {id=WavePlayer_8, properties {soundWaveAssetPtr=&/sword_slash_04}}
 {id=WavePlayer_9, properties {soundWaveAssetPtr=&/sword_slash_05}}
]
properties {
 attenuationSettings=&Fort_Weapon_Melee_Sml_Att
 maxDistance=2800
 soundClassObject=&Fort_WeaponMelee_PSM
 volumeMultiplier=0.8
 attenuationOverrides {distanceAlgorithm=NaturalSound, falloffDistance=5000, attenuationShapeExtents {x=1000}}
 concurrencyOverrides {maxCount=2, resolutionRule=StopOldest}
 concurrencySet [&Player_Melee_Weapon_Conc]
}
