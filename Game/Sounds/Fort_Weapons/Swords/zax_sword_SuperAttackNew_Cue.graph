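# Sound cue for the super attack of the zax sword melee weapon, playing a whoosh sound.
# A single wave player feeds the output directly, with no randomization or mixing, and the cue raises pitch slightly.
# It uses small melee weapon attenuation and the melee weapon sound class, and overrides concurrency so new instances are prevented at the limit rather than cutting off playing ones.
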SoundCue /Game/Sounds/Fort_Weapons/Swords/zax_sword_SuperAttackNew_Cue
root=WavePlayer_0
nodes [{id=WavePlayer_0, properties {soundWaveAssetPtr=&/zax_sword_whooshes_01_1}}]
properties {
 attenuationSettings=&Fort_Weapon_Melee_Sml_Att
 bOverrideConcurrency=true
 maxDistance=2800
 pitchMultiplier=1.1
 soundClassObject=&Fort_WeaponMelee_PSM
 concurrencyOverrides {resolutionRule=PreventNew}
}
